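# Physics asset for the IvoryCascade vehicle body mesh in the vehicle cosmetics content, giving it a simplified collision shape.
# It holds a single body on the root bone, made of one box offset forward and upward from the bone origin to approximate the whole body.
# It defines no constraints, so there are no articulated or separately simulated parts.
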PhysicsAsset /VehicleCosmetics/Bodies/IvoryCascade/Meshes/Body_IvoryCascade_PhysicsDM
bodies {root [{shapeType=Box, extentX=444, extentY=224, extentZ=108, shapeName=root_box, center=<30 0 92>}]}
constraints []
